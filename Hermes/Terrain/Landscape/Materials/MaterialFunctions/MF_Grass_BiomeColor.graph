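# Landscape material function that recolors grass per biome by overriding the base color of incoming material attributes, behind static switches.
# Terrain mask channels blend noise-varied palettes for grassland, mountains, wetlands, forests and boreal fall, with world-height wetland shading and heightmap-driven red-forest tips.
# A debug scalar blends these palettes toward simpler tints multiplied over reroute inputs; the edge-color chain is unconnected.
# A separate mobile output uses a smaller boreal tint chain.
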MaterialFunction /Hermes/Terrain/Landscape/Materials/MaterialFunctions/MF_Grass_BiomeColor
graph {
 properties {libraryCategoriesText [FN-Landscape]}
 nodes [
  FunctionInput() {bUsePreviewValueAsDefault=true, inputName=TerrainMasks_QuadD, inputType=FunctionInput_Vector4, sortPriority=40}
  n1 = NamedRerouteUsage()
  n2 = NamedRerouteUsage()
  FunctionOutput(None=SetMaterialAttributes!(
    MaterialAttributes=NamedRerouteUsage()
    BaseColor=StaticSwitchParameter(True=@n1, False=@n2) {defaultValue=true, parameterName=UseHermesBIomeColors}
   ) {attributeSetTypes [#1]})
  n3 = NamedRerouteUsage()
  n4 = NamedRerouteUsage()
  FunctionOutput(None=StaticSwitchParameter(True=SetMaterialAttributes(MaterialAttributes=@n3, BaseColor=@n4) {attributeSetTypes [#1]}, False=@n3) {defaultValue=true, parameterName=UseHeliosBIomeColors}) {outputName=Result Mobile, sortPriority=1}
  n5 = NamedRerouteUsage()
  n6 = NamedRerouteDeclaration(None=@n5) {name=NoBiomeColor, nodeColor=<rgba 0.5 0 1 1>}
  n7 = NamedRerouteUsage()
  n8 = Saturate(None=Multiply(A=VectorParameter() {group=BiomeColors, parameterName=GrassColor_Mountains, defaultValue=<rgba 0.55 0.55 0.55 1>}.RGB, B=@n7))
  n9 = NamedRerouteUsage()
  n10 = NamedRerouteUsage()
  n11 = Saturate(None=Multiply(A=VectorParameter() {group=BiomeColors, parameterName=GrassColor_Wetlands, defaultValue=<rgba 0.4511 0.7 0.6791 1>}.RGB, B=@n10))
  n12 = Saturate(None=Multiply(A=VectorParameter() {group=BiomeColors, parameterName=GrassColor_Wetlands2, defaultValue=<rgba 0.9667 1.5 1.4552 1>}.RGB, B=@n10))
  n13 = ScalarParameter() {
   defaultValue=-200
   desc="This offsets the ocean level used by the sand. Positive numbers go lower, negative numbers make it go higher."
   group=BiomeColors
   parameterName=GrassColor_Wetlands_Level
  }
  n14 = Add(A=@n13, B=WorldPosition().Z)
  n15 = ScalarParameter() {group=BiomeColors, parameterName=GrassColor_Wetlands_Level_Add}
  n16 = Saturate(None=Divide(A=Add(A=@n14, B=@n15), B=ScalarParameter() {defaultValue=400, group=BiomeColors, parameterName=GrassColor_Wetlands_Level_Div}))
  n17 = NamedRerouteUsage()
  n18 = VectorParameter() {group=BiomeColors, parameterName=GrassColor_MossyForest, defaultValue=<rgba 0 0.75 0.3007 1>}
  n19 = NamedRerouteUsage()
  n20 = NamedRerouteUsage()
  n21 = LinearInterpolate(
   A=LinearInterpolate(A=LinearInterpolate(A=@n6, B=@n8, Alpha=@n9), B=LinearInterpolate(A=@n11, B=@n12, Alpha=@n16), Alpha=@n17)
   B=Saturate(None=Multiply(A=@n18.RGB, B=@n19))
   Alpha=@n20
  )
  n22 = NamedRerouteUsage()
  n23 = Saturate(None=Multiply(A=VectorParameter() {group=BiomeColors, parameterName=GrassColor_BorealFall, defaultValue=<rgba 1.5 0.5556 0.3 1>}.RGB, B=@n22))
  n24 = NamedRerouteUsage()
  n25 = Multiply(
   A=VectorParameter() {group=BiomeColors, parameterName=GrassColor_MagicalForest, defaultValue=<rgba 0.98 0.56 1.4 1>}.RGB
   B=NamedRerouteUsage()
  )
  n26 = NamedRerouteUsage()
  n27 = VectorParameter() {group=BiomeColors, parameterName=GrassColor_RedForest, defaultValue=<rgba 2.5 0.25 0.3465 1>}
  n28 = NamedRerouteUsage()
  n29 = NamedRerouteUsage()
  n30 = LinearInterpolate(
   A=LinearInterpolate(A=LinearInterpolate(A=@n21, B=@n23, Alpha=@n24), B=Saturate(None=@n25), Alpha=@n26)
   B=Saturate(None=Multiply(A=@n27.RGB, B=@n28))
   Alpha=@n29
  )
  n31 = NamedRerouteUsage()
  n32 = VectorParameter() {group=Grass, parameterName=GrassEdgeColor_Grasslands, defaultValue=<rgba 2 1.0824 0.5 1>}
  n33 = VectorParameter() {group=BiomeColors, parameterName=GrassEdgeColor_Mountains, defaultValue=<rgba 1 1 1 1>}
  n34 = LinearInterpolate(A=Saturate(None=Multiply(A=@n31, B=@n32.RGB)), B=Saturate(None=Multiply(A=@n7, B=@n33.RGB)), Alpha=@n9)
  n35 = Saturate(None=Multiply(A=@n10, B=VectorParameter() {group=BiomeColors, parameterName=GrassEdgeColor_Wetlands, defaultValue=<rgba 1 1 1 1>}.RGB))
  n36 = Multiply(A=@n19, B=VectorParameter() {group=BiomeColors, parameterName=GrassEdgeColor_MossyForest, defaultValue=<rgba 1 1 1 1>}.RGB)
  n37 = VectorParameter() {group=BiomeColors, parameterName=GrassEdgeColor_BorealFall, defaultValue=<rgba 1 1 1 1>}
  n38 = LinearInterpolate(
   A=LinearInterpolate(A=LinearInterpolate(A=@n34, B=@n35, Alpha=@n17), B=Saturate(None=@n36), Alpha=@n20)
   B=Saturate(None=Multiply(A=@n22, B=@n37.RGB))
   Alpha=@n24
  )
  n39 = Saturate(None=Multiply(A=NamedRerouteUsage(), B=VectorParameter() {group=BiomeColors, parameterName=GrassEdgeColor_MagicalForest, defaultValue=<rgba 1 1 1 1>}.RGB))
  n40 = Multiply(A=@n28, B=VectorParameter() {group=BiomeColors, parameterName=GrassEdgeColor_RedForest, defaultValue=<rgba 1 1 1 1>}.RGB)
  LinearInterpolate(A=@n30, B=LinearInterpolate(A=LinearInterpolate(A=@n38, B=@n39, Alpha=@n26), B=Saturate(None=@n40), Alpha=@n29), Alpha=NamedRerouteUsage())
  n41 = VectorParameter() {group=BiomeColors, parameterName=RF_TipColor, defaultValue=<rgba 0.2 0.0423 0.01 1>}
  n42 = VectorParameter() {group=BiomeColors, parameterName=RF_CA, defaultValue=<rgba 0.35 0.0318 0.0245 1>}
  n43 = NamedRerouteUsage()
  n44 = LinearInterpolate(A=@n42.RGB, B=VectorParameter() {group=BiomeColors, parameterName=RF_CB, defaultValue=<rgba 0.069 0.08 0.0125 1>}.RGB, Alpha=@n43)
  n45 = Desaturation(None=@n42.RGB, Fraction=ScalarParameter() {defaultValue=0.14, group=Grass, parameterName=Noise02_Desaturation})
  n46 = NamedRerouteUsage()
  n47 = LinearInterpolate(
   A=VectorParameter() {group=BiomeColors, parameterName=RF_CD, defaultValue=<rgba 0.2 0.0278 0.02 1>}.RGB
   B=LinearInterpolate(
    A=LinearInterpolate(A=@n44, B=@n45, Alpha=@n46)
    B=VectorParameter() {group=BiomeColors, parameterName=RF_CC, defaultValue=<rgba 0.0842 0.15 0.0125 1>}.RGB
    Alpha=NamedRerouteUsage()
   )
   Alpha=NamedRerouteUsage()
  )
  Multiply(A=@n41.RGB, B=@n47)
  n48 = MaterialFunctionCall(Float4=FunctionInput() {bUsePreviewValueAsDefault=true, inputName=TerrainMasks_QuadA, inputType=FunctionInput_Vector4, sortPriority=10}) {materialFunction=&BreakOutFloat4Components}
  NamedRerouteDeclaration(None=@n48.R) {name=BM_Mountains, nodeColor=<rgba 0.5 0 1 1>}
  NamedRerouteDeclaration(None=@n48.G) {name=BM_Wetlands, nodeColor=<rgba 0.5 0 1 1>}
  n49 = VectorParameter() {group=BiomeColors, parameterName=MT_CA, defaultValue=<rgba 0.025 0.1 0.0274 1>}
  n50 = LinearInterpolate(
   A=@n49.RGB
   B=VectorParameter() {group=BiomeColors, parameterName=MT_CB, defaultValue=<rgba 0.045 0.15 0.0502 1>}.RGB
   Alpha=NamedRerouteUsage()
  )
  n51 = Desaturation(None=@n49.RGB, Fraction=ScalarParameter() {defaultValue=0.14, group=Grass, parameterName=Noise02_Desaturation})
  n52 = NamedRerouteUsage()
  n53 = LinearInterpolate(
   A=VectorParameter() {group=BiomeColors, parameterName=MT_CD, defaultValue=<rgba 0.0253 0.09 0.0345 1>}.RGB
   B=LinearInterpolate(
    A=LinearInterpolate(A=@n50, B=@n51, Alpha=@n52)
    B=VectorParameter() {group=BiomeColors, parameterName=MT_CC, defaultValue=<rgba 0.0512 0.15 0.0416 1>}.RGB
    Alpha=NamedRerouteUsage()
   )
   Alpha=NamedRerouteUsage()
  )
  n54 = NamedRerouteUsage()
  n55 = VectorParameter() {group=BiomeColors, parameterName=GL_CA, defaultValue=<rgba 0.0845 0.1732 0.0065 1>}
  n56 = LinearInterpolate(
   A=@n55.RGB
   B=VectorParameter() {group=BiomeColors, parameterName=GL_CB, defaultValue=<rgba 0.0651 0.1137 0.0105 1>}.RGB
   Alpha=NamedRerouteUsage()
  )
  n57 = Desaturation(None=@n55.RGB, Fraction=ScalarParameter() {defaultValue=0.14, group=Grass, parameterName=Noise02_Desaturation})
  n58 = NamedRerouteUsage()
  n59 = LinearInterpolate(
   A=VectorParameter() {group=BiomeColors, parameterName=GL_CD, defaultValue=<rgba 0.0503 0.13 0.0125 1>}.RGB
   B=LinearInterpolate(
    A=LinearInterpolate(A=@n56, B=@n57, Alpha=@n58)
    B=VectorParameter() {group=BiomeColors, parameterName=GL_CC, defaultValue=<rgba 0.1263 0.144 0.0199 1>}.RGB
    Alpha=NamedRerouteUsage()
   )
   Alpha=NamedRerouteUsage()
  )
  n60 = VectorParameter() {group=BiomeColors, parameterName=WL_CA, defaultValue=<rgba 0.0381 0.17 0.0059 1>}
  n61 = LinearInterpolate(
   A=@n60.RGB
   B=VectorParameter() {group=BiomeColors, parameterName=WL_CB, defaultValue=<rgba 0.0664 0.12 0.0248 1>}.RGB
   Alpha=NamedRerouteUsage()
  )
  n62 = Desaturation(None=@n60.RGB, Fraction=ScalarParameter() {defaultValue=0.14, group=Grass, parameterName=Noise02_Desaturation})
  n63 = NamedRerouteUsage()
  n64 = LinearInterpolate(
   A=VectorParameter() {group=BiomeColors, parameterName=WL_CD, defaultValue=<rgba 0.0159 0.09 0.0042 1>}.RGB
   B=LinearInterpolate(
    A=LinearInterpolate(A=@n61, B=@n62, Alpha=@n63)
    B=VectorParameter() {group=BiomeColors, parameterName=WL_CC, defaultValue=<rgba 0.0041 0.06 0.013 1>}.RGB
    Alpha=NamedRerouteUsage()
   )
   Alpha=NamedRerouteUsage()
  )
  n65 = Saturate(None=Multiply(A=VectorParameter() {group=BiomeColors, parameterName=WL_LowColor, defaultValue=<rgba 1.8 1.5912 0.5084 1>}.RGB, B=@n64))
  n66 = VectorParameter() {group=BiomeColors, parameterName=MF_CA, defaultValue=<rgba 0.0488 0.1 0.0038 1>}
  n67 = LinearInterpolate(
   A=LinearInterpolate(
    A=@n66.RGB
    B=VectorParameter() {group=BiomeColors, parameterName=MF_CB, defaultValue=<rgba 0.0229 0.04 0.0037 1>}.RGB
    Alpha=NamedRerouteUsage()
   )
   B=Desaturation(None=@n66.RGB, Fraction=ScalarParameter() {defaultValue=0.14, group=Grass, parameterName=Noise02_Desaturation})
   Alpha=NamedRerouteUsage()
  )
  n68 = VectorParameter() {group=BiomeColors, parameterName=MF_CC, defaultValue=<rgba 0.0088 0.01 0.0014 1>}
  n69 = NamedRerouteUsage()
  n70 = LinearInterpolate(
   A=LinearInterpolate(A=LinearInterpolate(A=@n59, B=@n53, Alpha=@n54), B=LinearInterpolate(A=@n65, B=@n64, Alpha=@n16), Alpha=NamedRerouteUsage())
   B=LinearInterpolate(
    A=VectorParameter() {group=BiomeColors, parameterName=MF_CD, defaultValue=<rgba 0.0193 0.05 0.0048 1>}.RGB
    B=LinearInterpolate(A=@n67, B=@n68.RGB, Alpha=@n69)
    Alpha=NamedRerouteUsage()
   )
   Alpha=NamedRerouteUsage()
  )
  n71 = VectorParameter() {group=BiomeColors, parameterName=BF_CA, defaultValue=<rgba 0.1415 0.15 0.0123 1>}
  n72 = LinearInterpolate(
   A=@n71.RGB
   B=VectorParameter() {group=BiomeColors, parameterName=BF_CB, defaultValue=<rgba 0.0517 0.1 0.0114 1>}.RGB
   Alpha=NamedRerouteUsage()
  )
  n73 = Desaturation(None=@n71.RGB, Fraction=ScalarParameter() {defaultValue=0.14, group=Grass, parameterName=Noise02_Desaturation})
  n74 = NamedRerouteUsage()
  n75 = LinearInterpolate(
   A=VectorParameter() {group=BiomeColors, parameterName=BF_CD, defaultValue=<rgba 0.11 0.0927 0.0165 1>}.RGB
   B=LinearInterpolate(
    A=LinearInterpolate(A=@n72, B=@n73, Alpha=@n74)
    B=VectorParameter() {group=BiomeColors, parameterName=BF_CC, defaultValue=<rgba 0.1 0.0916 0.02 1>}.RGB
    Alpha=NamedRerouteUsage()
   )
   Alpha=NamedRerouteUsage()
  )
  n76 = NamedRerouteUsage()
  n77 = VectorParameter() {group=BiomeColors, parameterName=MG_CD, defaultValue=<rgba 0.1275 0.03 0.2 1>}
  n78 = VectorParameter() {group=BiomeColors, parameterName=MG_CA, defaultValue=<rgba 0.0845 0.1732 0.0065 1>}
  n79 = VectorParameter() {group=BiomeColors, parameterName=MG_CB, defaultValue=<rgba 0.1379 0.015 0.15 1>}
  n80 = NamedRerouteUsage()
  n81 = ScalarParameter() {defaultValue=0.14, group=Grass, parameterName=Noise02_Desaturation}
  n82 = LinearInterpolate(
   A=LinearInterpolate(A=LinearInterpolate(A=@n78.RGB, B=@n79.RGB, Alpha=@n80), B=Desaturation(None=@n78.RGB, Fraction=@n81), Alpha=NamedRerouteUsage())
   B=VectorParameter() {group=BiomeColors, parameterName=MG_CC, defaultValue=<rgba 0.4411 0.072 0.9 1>}.RGB
   Alpha=NamedRerouteUsage()
  )
  n83 = NamedRerouteUsage()
  n84 = NamedRerouteUsage()
  n85 = ScalarParameter() {defaultValue=1, group=BiomeColors, parameterName=RF_Height_Power}
  n86 = Saturate(None=Multiply(A=Power(Base=@n84, Exp=@n85), B=ScalarParameter() {defaultValue=1, group=BiomeColors, parameterName=RF_Height_Mult}))
  n87 = LinearInterpolate(
   A=LinearInterpolate(A=LinearInterpolate(A=@n70, B=@n75, Alpha=@n76), B=LinearInterpolate(A=@n77.RGB, B=@n82, Alpha=@n83), Alpha=NamedRerouteUsage())
   B=LinearInterpolate(A=@n47, B=@n41.RGB, Alpha=Multiply(A=@n86, B=@n43))
   Alpha=NamedRerouteUsage()
  )
  n88 = ScalarParameter() {parameterName=MF_Grass_colortypedebugz}
  NamedRerouteDeclaration(None=ComponentMask(None=LinearInterpolate(A=@n87, B=@n30, Alpha=@n88)) {b=true, g=true, r=true}) {name=OutputBasecolor, nodeColor=<rgba 0.5 0 1 1>}
  n89 = FunctionInput() {bUsePreviewValueAsDefault=true, inputName=GrassNoise, inputType=FunctionInput_Vector4, sortPriority=50, previewValue=<0 0 0 0>}
  NamedRerouteDeclaration(None=@n89) {name=GrassNoiseVec4, nodeColor=<rgba 0.5 0 1 1>}
  n90 = MaterialFunctionCall(Float4=FunctionInput() {bUsePreviewValueAsDefault=true, inputName=TerrainMasks_QuadB, inputType=FunctionInput_Vector4, sortPriority=20}) {materialFunction=&BreakOutFloat4Components}
  NamedRerouteDeclaration(None=@n90.R) {name=BM_MagicalForest, nodeColor=<rgba 0.5 0 1 1>}
  n91 = GetMaterialAttributes(None=FunctionInput(Preview=SetMaterialAttributes!(MaterialAttributes)) {bUsePreviewValueAsDefault=true, inputName=Material, inputType=FunctionInput_MaterialAttributes}) {attributeGetTypes [#1, #2]}
  n92 = NamedRerouteDeclaration(None=@n91.BaseColor) {name=Input_BaseColor, nodeColor=<rgba 0.5 0 1 1>}
  NamedRerouteDeclaration(None=Desaturation(None=@n92, Fraction=Constant() {r=0.5})) {name=BaseColor_DesatHalf, nodeColor=<rgba 0.5 0 1 1>}
  NamedRerouteDeclaration(None=@n90.G) {name=BM_RedForest, nodeColor=<rgba 0.5 0 1 1>}
  n93 = Constant()
  n94 = Constant() {r=1.5}
  n95 = MaterialFunctionCall!(Coordinates, R_01_Scale, R_02_Scale, R_03_Scale, R_04_Scale, G_01_Scale, G_02_Scale, G_03_Scale, G_04_Scale) {materialFunction=/Game/Athena/Helios/Environments/Landscape/Materials/MaterialFunctions_Utilities/MF_VariationMasks2}
  NamedRerouteDeclaration(None=Saturate(None=LinearInterpolate(A=@n93, B=@n94, Alpha=@n95.G_01))) {name=Noise01, nodeColor=<rgba 0.5 0 1 1>}
  n96 = Constant() {r=1.5}
  n97 = Constant() {r=-9}
  NamedRerouteDeclaration(None=Saturate(None=LinearInterpolate(A=@n96, B=@n97, Alpha=@n95.G_02))) {name=Noise02, nodeColor=<rgba 0.5 0 1 1>}
  n98 = Constant() {r=-0.25}
  n99 = Constant() {r=1.5}
  n100 = NamedRerouteDeclaration(None=Saturate(None=LinearInterpolate(A=@n98, B=@n99, Alpha=@n95.G_03))) {name=Noise03_PreMult, nodeColor=<rgba 0.5 0 1 1>}
  n101 = Constant() {r=0.65}
  NamedRerouteDeclaration(None=Saturate(None=Multiply(A=@n100, B=@n101))) {name=Noise03, nodeColor=<rgba 0.5 0 1 1>}
  NamedRerouteDeclaration(None=@n48.B) {name=BM_MossyForest, nodeColor=<rgba 0.5 0 1 1>}
  n102 = Constant() {r=4}
  NamedRerouteDeclaration(None=Saturate(None=Multiply(A=@n95.G_04, B=@n102))) {name=Noise04, nodeColor=<rgba 0.5 0 1 1>}
  NamedRerouteDeclaration(None=@n48.A) {name=BM_BorealFall, nodeColor=<rgba 0.5 0 1 1>}
  NamedRerouteDeclaration(None=FunctionInput(Preview=Constant()) {bUsePreviewValueAsDefault=true, inputName=GrassEdgeMask, inputType=FunctionInput_Scalar, sortPriority=1}) {name=GrassEdgeMask, nodeColor=<rgba 0.5 0 1 1>}
  NamedRerouteDeclaration(None=Desaturation(None=@n92, Fraction=Constant() {r=1})) {name=BaseColorDesat, nodeColor=<rgba 0.5 0 1 1>}
  n103 = MaterialFunctionCall(Float4=FunctionInput() {bUsePreviewValueAsDefault=true, inputName=TerrainMasks_QuadC, inputType=FunctionInput_Vector4, sortPriority=30}) {materialFunction=&BreakOutFloat4Components}
  NamedRerouteDeclaration(None=@n103.R) {name=Biome_Hades, nodeColor=<rgba 0.5 0 1 1>}
  NamedRerouteDeclaration(None=@n103.G) {name=Biome_Olympus, nodeColor=<rgba 0.5 0 1 1>}
  NamedRerouteDeclaration(None=@n91.MaterialAttributes) {name=InputMaterial, nodeColor=<rgba 0.5 0 1 1>}
  NamedRerouteDeclaration(None=@n91.OpacityMask) {name=Heightmap, nodeColor=<rgba 0.5 0 1 1>}
  n104 = NamedRerouteUsage()
  n105 = Saturate(None=Multiply(
    A=VectorParameter() {group=BiomeColors, parameterName=GrassColor_Boreal_Chaparral_Mobile, defaultValue=<rgba 1.6 1.2 1.7 1>}.RGB
    B=NamedRerouteUsage()
   ))
  n106 = NamedRerouteUsage()
  n107 = Multiply(
   A=VectorParameter() {group=BiomeColors, parameterName=GrassColor_Boreal_Tundra, defaultValue=<rgba 0.4 0.5056 2 1>}.RGB
   B=NamedRerouteUsage()
  )
  n108 = VectorParameter() {group=BiomeColors, parameterName=GrassColor_Boreal_Forest, defaultValue=<rgba 0.25 0.316 1.25 1>}
  n109 = NamedRerouteUsage()
  n110 = LinearInterpolate(
   A=LinearInterpolate(A=LinearInterpolate(A=@n104, B=@n105, Alpha=@n106), B=Saturate(None=@n107), Alpha=NamedRerouteUsage())
   B=Saturate(None=Multiply(A=@n108.RGB, B=@n109))
   Alpha=NamedRerouteUsage()
  )
  n111 = Saturate(None=Multiply(
    A=VectorParameter() {group=BiomeColors, parameterName=GrassColor_Boreal_Mountain, defaultValue=<rgba 0.25 0.316 1.25 1>}.RGB
    B=NamedRerouteUsage()
   ))
  n112 = NamedRerouteUsage()
  NamedRerouteDeclaration(None=ComponentMask(None=LinearInterpolate(A=@n110, B=@n111, Alpha=@n112)) {b=true, g=true, r=true}) {name=OutputBasecolorMobile, nodeColor=<rgba 0.5 0 1 1>}
  NamedRerouteDeclaration(None=@n103.B) {name=Biome_Arid_Desert, nodeColor=<rgba 0.5 0 1 1>}
  NamedRerouteDeclaration(None=@n103.A) {name=Biome_Open_Desert, nodeColor=<rgba 0.5 0 1 1>}
  OneMinus(None=@n43)
  Saturate(None=Multiply(A=@n5, B=VectorParameter() {group=Grass, parameterName=GrassColor_Grasslands, defaultValue=<rgba 1 1 1 1>}.RGB))
 ]
}
